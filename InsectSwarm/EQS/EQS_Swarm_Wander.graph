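# EQS query that picks a wander destination for an insect swarm, favoring launchable pod plants and jungle ferns.
# Fallback options look for plant actors in a near box, then a wider box, skipping plants tagged in-use or regrowing.
# The last option drops plants for a coarse navmesh grid that excludes water.
# Every option filters out close points, checks pathfinding with a water-avoiding filter, and adds random scoring.
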EnvQuery /InsectSwarm/EQS/EQS_Swarm_Wander
options [
 {
  generator {
   name=FortQueryGenerator_PointsOnBuildingActors_0
   class=FortQueryGenerator_PointsOnBuildingActors
   properties {
    bAutoSortTests=true
    bCanRunAsync=false
    boundingBoxExtentXY=3500
    boundingBoxExtentZ=2000
    envQueryResultNormalizationOption=Default
    generateAround=&EnvQueryContext_Querier
    maxGeneratedPoints=10
    navDataOverrideContext=null
    optionName=""
    pointDensity=1000
    randomChanceToSkip=0
    actorLookupFilter {
     minDistanceFromBoundsCenter=0
     bounds {sphereRadius=0, boxExtent=<0 0 0>, origin=<0 0 0>}
     optionalSubclasses [
      /ChronoEnviroItems/LaunchablePlants/Gameplay/Actors/BGA_Chrono_LaunchablePlant_Parent.BGA_Chrono_LaunchablePlant_Parent_C
      /Game/Environments/Asteria/Foliage/JunglePlants/FernPlants/Blueprints/BP_JunglePlants_FernPlant_A.BP_JunglePlants_FernPlant_A_C
     ]
     requiredProperties []
     tagQuery {
      autoDescription=" NONE( Athena.AI.Status.InsectSwarm.PlantInUse, Chrono.EnvItem.PodPlant.InRegrowth )"
      tokenStreamVersion=0
      userDescription=""
      queryTokenStream [0,1,3,2,0,1]
      tagDictionary [{tagName=Athena.AI.Status.InsectSwarm.PlantInUse}, {tagName=Chrono.EnvItem.PodPlant.InRegrowth}]
     }
    }
    projectionData {
     bCanDisableTrace=true
     bCanProjectDown=true
     bCanTraceOnGeometry=true
     bCanTraceOnNavMesh=true
     bOnlyBlockingHits=true
     bTraceComplex=false
     extentX=0
     extentY=10
     extentZ=10
     navigationFilter=null
     postProjectionVerticalOffset=0
     projectDown=1024
     projectUp=1024
     serializedChannel=ECC_Visibility
     traceChannel=TraceTypeQuery1
     traceMode=None
     traceProfileName=None
     traceShape=Line
     versionNum=1
    }
   }
  }
  tests [
   EnvQueryTest_Distance_2 {
    class=EnvQueryTest_Distance
    properties {
     distanceTo=&EnvQueryContext_Querier
     filterType=Minimum
     floatValueMin=500
     scoringEquation=Linear
     scoringFactor=1
     testComment=""
     testMode=Distance2D
     testPurpose=Filter
    }
   }
   EnvQueryTest_Pathfinding_1 {
    class=EnvQueryTest_Pathfinding
    properties {filterClass=&NavFilter_NPC_AvoidAllWater_C}
   }
   FortQueryTest_Random_2 {class=FortQueryTest_Random, properties {floatValueMax=1, scoringFactor=1, testPurpose=FilterAndScore}}
  ]
 }
 {
  generator {
   name=FortQueryGenerator_PointsOnBuildingActors_1
   class=FortQueryGenerator_PointsOnBuildingActors
   properties {
    bAutoSortTests=true
    bCanRunAsync=false
    boundingBoxExtentXY=8500
    boundingBoxExtentZ=2000
    envQueryResultNormalizationOption=Default
    generateAround=&EnvQueryContext_Querier
    maxGeneratedPoints=10
    navDataOverrideContext=null
    optionName=""
    pointDensity=1000
    randomChanceToSkip=0
    actorLookupFilter {
     minDistanceFromBoundsCenter=0
     bounds {sphereRadius=0, boxExtent=<0 0 0>, origin=<0 0 0>}
     optionalSubclasses [
      &BGA_Chrono_LaunchablePlant_Parent_C
      &BP_JunglePlants_FernPlant_A_C
     ]
     requiredProperties []
     tagQuery {
      autoDescription=" NONE( Athena.AI.Status.InsectSwarm.PlantInUse, Chrono.EnvItem.PodPlant.InRegrowth )"
      tokenStreamVersion=0
      userDescription=""
      queryTokenStream [0,1,3,2,0,1]
      tagDictionary [{tagName=Athena.AI.Status.InsectSwarm.PlantInUse}, {tagName=Chrono.EnvItem.PodPlant.InRegrowth}]
     }
    }
    projectionData {
     bCanDisableTrace=true
     bCanProjectDown=true
     bCanTraceOnGeometry=true
     bCanTraceOnNavMesh=true
     bOnlyBlockingHits=true
     bTraceComplex=false
     extentX=0
     extentY=10
     extentZ=10
     navigationFilter=null
     postProjectionVerticalOffset=0
     projectDown=1024
     projectUp=1024
     serializedChannel=ECC_Visibility
     traceChannel=TraceTypeQuery1
     traceMode=None
     traceProfileName=None
     traceShape=Line
     versionNum=1
    }
   }
  }
  tests [
   EnvQueryTest_Distance_1 {
    class=EnvQueryTest_Distance
    properties {
     distanceTo=&EnvQueryContext_Querier
     filterType=Minimum
     floatValueMin=500
     scoringEquation=Linear
     scoringFactor=1
     testComment=""
     testMode=Distance2D
     testPurpose=Filter
    }
   }
   EnvQueryTest_Pathfinding_2 {
    class=EnvQueryTest_Pathfinding
    properties {filterClass=&NavFilter_NPC_AvoidAllWater_C}
   }
   FortQueryTest_Random_0 {class=FortQueryTest_Random, properties {floatValueMax=1, scoringFactor=1, testPurpose=FilterAndScore}}
  ]
 }
 {
  generator {
   name=EnvQueryGenerator_SimpleGrid_0
   class=EnvQueryGenerator_SimpleGrid
   properties {
    gridSize=5000
    spaceBetween=2000
    projectionData {
     extentX=10
     navigationFilter=&NavFilter_NPC_ExcludeAllWater_C
     postProjectionVerticalOffset=30
     projectDown=5000
     projectUp=5000
    }
   }
  }
  tests [
   EnvQueryTest_Distance_0 {
    class=EnvQueryTest_Distance
    properties {
     distanceTo=&EnvQueryContext_Querier
     filterType=Minimum
     floatValueMin=1500
     scoringEquation=Linear
     scoringFactor=1
     testComment=""
     testMode=Distance3D
     testPurpose=Filter
    }
   }
   EnvQueryTest_Pathfinding_0 {
    class=EnvQueryTest_Pathfinding
    properties {filterClass=&NavFilter_NPC_AvoidAllWater_C}
   }
   FortQueryTest_Random_1 {class=FortQueryTest_Random, properties {floatValueMax=1, scoringFactor=1, testPurpose=FilterAndScore}}
  ]
 }
]
